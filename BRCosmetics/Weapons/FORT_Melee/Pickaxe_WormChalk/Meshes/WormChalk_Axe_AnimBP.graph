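# Animation Blueprint for the WormChalk pickaxe mesh that makes its right upper and lower arm bones follow those of the owning character.
# BeginPlay caches the owner's skeletal mesh and sets a validity flag; each animation update reads both bone transforms from it.
# Two Modify Bone nodes, enabled by that flag, replace upper-arm position and rotation in world space, then lower-arm rotation only, in parent-bone space.
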AnimBlueprint /BRCosmetics/Weapons/FORT_Melee/Pickaxe_WormChalk/Meshes/WormChalk_Axe_AnimBP
target_skeleton=&/WormChalk_Axe_Skeleton
graphs [
 AnimGraph {nodes [
   AnimGraphNode_SubInput_0 = Input Pose()
   n1 = Get!() -> (upperarm_r_transform_Location:Vector, upperarm_r_transform_Rotation:Rotator, upperarm_r_transform_Scale:Vector)
   n2 = "Transform (Modify) Bone - Bone: upperarm_r"!(
    ComponentPose=Local To Component(LocalPose=@AnimGraphNode_SubInput_0)
    bAlphaBoolEnabled:bool=Get isValid() -> (isValid:bool)
    Translation=@n1
    Rotation=@n1.upperarm_r_transform_Rotation
    Scale="1,1,1"
   ) -> (Pose) {
    alphaInputType=Bool
    rotationMode=BMM_Replace
    rotationSpace=BCS_WorldSpace
    translationMode=BMM_Replace
    translationSpace=BCS_WorldSpace
    alphaBoolBlend {blendOption=Linear}
    boneToModify {boneName=upperarm_r}
   }
   n3 = Get!() -> (lowerarm_r_transform_Location:Vector, lowerarm_r_transform_Rotation:Rotator, lowerarm_r_transform_Scale:Vector)
   n4 = "Transform (Modify) Bone - Bone: lowerarm_r"!(ComponentPose=@n2, bAlphaBoolEnabled:bool=Get isValid(), Translation=@n3, Rotation=@n3.lowerarm_r_transform_Rotation, Scale="1,1,1") -> (Pose) {
    alphaInputType=Bool
    rotationMode=BMM_Replace
    rotationSpace=BCS_ParentBoneSpace
    translationSpace=BCS_WorldSpace
    alphaBoolBlend {blendOption=Linear}
    boneToModify {boneName=lowerarm_r}
   }
   Output Pose(Result=Component To Local(ComponentPose=@n4)) {layerGroup=None}
  ]}
 EventGraph {nodes [
   Event BlueprintUpdateAnimation()
   n1 = Get Character_SkeletalMesh_ref() -> (Character_SkeletalMesh_ref:Skeletal Mesh Component Object Reference)
   > Set upperarm_r_transform(upperarm_r_transform:Transform=GetBoneTransform!(self:Skinned Mesh Component Object Reference=@n1, InBoneName:Name=upperarm_r, TransformSpace:ERelativeTransformSpace Enum=RTS_World) -> (ReturnValue:Transform)) -> (Output_Get:Transform)
   > Set lowerarm_r_transform(lowerarm_r_transform:Transform=GetBoneTransform(self=@n1, InBoneName=lowerarm_r, TransformSpace=RTS_ParentBoneSpace)) -> (Output_Get:Transform)
   Event BlueprintBeginPlay() {is=title K2Node_Event}
   > n2 = Set Character_SkeletalMesh_ref(Character_SkeletalMesh_ref:Skeletal Mesh Component Object Reference=GetOwnerSkeletalMeshComponent()) -> (Output_Get:Skeletal Mesh Component Object Reference)
   > n3 = Is Valid(InputObject=@n2)
   Set isValid(execute=@n3.Is Not Valid, isValid:bool=false) -> (Output_Get:bool)
   Set isValid(execute=@n3.Is Valid, isValid=true)
  ]}
]
variables [
 Character_SkeletalMesh_ref {type=SkeletalMeshComponent}
 isValid {type=boolean}
 lowerarm_r_transform {type=Transform, default {location=<-26.6514 0 0>, rotation=<rot 0 0 0>, scale=<1 1 1>}}
 upperarm_r_transform {type=Transform, default {location=<0 0 0>, rotation=<rot 0 0 0>, scale=<1 1 1>}}
]
